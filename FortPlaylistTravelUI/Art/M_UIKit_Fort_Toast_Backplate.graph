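# Translucent UI material for a toast notification backplate in the playlist travel UI, with a colored glow and progress bar.
# Rounded-rectangle masks shape the plate, outline and bar, while a scaled radial gradient supplies the glow.
# Progress Amount sizes the soft-edged cyan-to-white bar, and Progress Bar Visibility scales its blend.
# Named reroutes carry masks, an eased Animation Control and a time-based ping-pong pulse, and compose final color and opacity.
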Material /FortPlaylistTravelUI/Art/M_UIKit_Fort_Toast_Backplate
graph {
 properties {bAutomaticallySetUsageInEditor=false, blendMode=BLEND_Translucent, materialDomain=MD_UI, usedShadingModels=Default Lit}
 nodes [
  n1 = NamedRerouteUsage()
  n2 = MaterialFunctionCall(
   UVs=MaterialFunctionCall() {materialFunction=&GetUserInterfaceUV}.Normalized UV
   Texture Scale=AppendVector(A=LinearInterpolate(Alpha=@n1) {constA=1.5}, B=Constant() {r=1})
  ) {materialFunction=&ScaleUVsByCenter}
  n3 = LinearInterpolate(Alpha=NamedRerouteUsage()) {constB=-0.5}
  n4 = Constant() {r=-0.4}
  n5 = MaterialFunctionCall(UV=Subtract(
    A=MaterialFunctionCall(UV=@n2.UVs) {materialFunction=&MF_UI_UVScaling}.ScaleX
    B=AppendVector(A=@n3, B=@n4)
   ) {constB=0.5}) {materialFunction=&MF_UI_RadialGradient}
  n6 = NamedRerouteUsage()
  NamedRerouteDeclaration(None=Multiply(A=@n5.Smooth, B=LinearInterpolate(Alpha=@n6) {constA=0.5, constB=0.2}) {constB=0.2}) {name=Glow Mask, nodeColor=<rgba 1 0.3047 0 1>}
  NamedRerouteDeclaration(None=MaterialFunctionCall!(Value=Time(), Scale=Constant() {r=0.4}, Height=Constant() {r=1}) {materialFunction=/FortUILibrary/MaterialFunctions/Basics/MF_UI_PingPong}.Result) {name=Pulse Animation, nodeColor=<rgba 0.5469 0 1 1>}
  NamedRerouteDeclaration(None=MaterialFunctionCall(Size=Constant() {r=1}, Corner Radius=Constant() {r=10}, Radius in Pixels=Constant() {r=1}, Outline Thickness=Constant() {r=2}) {materialFunction=&MF_UI_CoreUI_RoundedMask}.Outline Mask) {name=Top Gradient Fade Mask, nodeColor=<rgba 0.7109 1 0 1>}
  NamedRerouteDeclaration(None=LinearInterpolate(B=@n5.Smooth, Alpha=NamedRerouteUsage())) {name=Top Glow, nodeColor=<rgba 0.9688 0 1 1>}
  n7 = Subtract(
   A=ComponentMask(None=MaterialFunctionCall() {materialFunction=&GetUserInterfaceUV}.Normalized UV) {g=true}
   B=OneMinus(None=NamedRerouteUsage())
  ) {constB=0.9}
  NamedRerouteDeclaration(None=Ceil(None=Saturate(None=@n7))) {name=Progress Bar Background Mask, nodeColor=<rgba 1 0.75 0 1>}
  n8 = NamedRerouteUsage()
  n9 = MaterialFunctionCall() {materialFunction=&GetUserInterfaceUV}
  n10 = MaterialFunctionCall!(Value PX=Constant() {r=4}, UV Pixel Size, X or Y=Constant() {r=1}) {materialFunction=/FortUILibrary/MaterialFunctions/Basics/MF_UI_PixelsToNormalized}
  n11 = AppendVector(
   A=Multiply(A=ScalarParameter() {defaultValue=0.6, parameterName=Progress Amount}) {constB=-0.5}
   B=Multiply(A=OneMinus(None=@n10.Normalized)) {constB=0.5}
  )
  n12 = NamedRerouteDeclaration(None=ScalarParameter() {defaultValue=0.6, parameterName=Progress Amount}) {name=Progress Amount, nodeColor=<rgba 0.0078 0 1 1>}
  n13 = MaterialFunctionCall(
   UV=Saturate(None=Subtract(A=@n9.Normalized UV, B=@n11) {constB=0.9})
   Size=AppendVector(A=OneMinus(None=@n12), B=NamedRerouteDeclaration(None=@n10.Normalized) {name=Normalized Pixel Width, nodeColor=<rgba 0 1 0.8125 1>})
   Corner Radius=Constant() {r=0.5}
  ) {materialFunction=&MF_UI_CoreUI_RoundedMask}
  n14 = MaterialFunctionCall("Value (0-1)"=MaterialFunctionCall(DistanceField=@n13.SDF, Smoothing=LinearInterpolate(Alpha=@n8) {constA=0.4, constB=0.5}) {materialFunction=&MF_UI_AntiAliasing}.Inner) {materialFunction=&MF_UI_EaseCurves}
  n15 = LinearInterpolate(Alpha=@n8) {constA=0.25, constB=0.4}
  NamedRerouteDeclaration(None=LinearInterpolate(
    A=Multiply(A=@n14.EaseInOutCubic, B=@n15) {constB=0.25}
    Alpha=MaterialFunctionCall(DistanceField=@n13.SDF) {materialFunction=&MF_UI_AntiAliasing}.Inner
   )) {name=Progress Bar Mask, nodeColor=<rgba 0 0.5312 1 1>}
  n16 = ComponentMask(None=MaterialFunctionCall() {materialFunction=&GetUserInterfaceUV}.Normalized UV) {r=true}
  n17 = Constant()
  n18 = OneMinus(None=ScalarParameter() {defaultValue=0.6, parameterName=Progress Amount})
  NamedRerouteDeclaration(None=LinearInterpolate(
    A=Constant3Vector() {constant=<rgba 0 0.8714 0.9823 1>}
    B=Constant3Vector() {constant=<rgba 1 1 1 0>}
    Alpha=MaterialFunctionCall(Input=@n16, Input Low=@n17, Input High=@n18) {materialFunction=&RemapValue_0-1}.Result
   )) {name=Progress Bar Color, nodeColor=<rgba 1 0.2344 0 1>}
  n19 = Multiply(A=Max(A=NamedRerouteUsage(), B=NamedRerouteUsage()), B=NamedRerouteUsage())
  n20 = MaterialFunctionCall("Value (0-1)"=ComponentMask(None=MaterialFunctionCall() {materialFunction=&GetUserInterfaceUV}.Normalized UV) {g=true}) {materialFunction=&MF_UI_EaseCurves}
  n21 = Saturate(None=Add(A=NamedRerouteUsage(), B=LinearInterpolate(Alpha=@n20.EaseInSine) {constA=0.7, constB=0.95}))
  n22 = Constant() {r=10}
  n23 = Constant() {r=1}
  NamedRerouteDeclaration(None=Multiply(
    A=Max(A=@n19, B=@n21)
    B=MaterialFunctionCall(Corner Radius=@n22, Radius in Pixels=@n23) {materialFunction=&MF_UI_CoreUI_RoundedMask}.Shape Mask
   )) {name=Final Opacity, nodeColor=<rgba 1 0 0.75 1>}
  n24 = LinearInterpolate(
   A=LinearInterpolate(
    A=Constant3Vector() {constant=<rgba 0.0065 0.006 0.0123 1>}
    B=VectorParameter() {parameterName=Glow Color, defaultValue=<rgba 1 0.147 0.1329 1>}.RGB
    Alpha=NamedRerouteUsage()
   )
   B=Constant3Vector() {constant=<rgba 0.0296 0.0296 0.0356 1>}
   Alpha=Multiply(A=NamedRerouteUsage(), B=NamedRerouteUsage())
  )
  n25 = NamedRerouteUsage()
  n26 = ScalarParameter() {defaultValue=1, parameterName=Progress Bar Visibility, sliderMax=1}
  n27 = Floor(None=NamedRerouteUsage())
  n28 = Multiply(
   A=NamedRerouteUsage()
   B=Multiply(
    A=NamedRerouteDeclaration(None=@n26) {name=Progress Bar Visibility, nodeColor=<rgba 0.2188 0 1 1>}
    B=LinearInterpolate(Alpha=@n27) {constA=1, constB=0}
   )
  )
  n29 = NamedRerouteUsage()
  n30 = NamedRerouteUsage()
  NamedRerouteDeclaration(None=LinearInterpolate(
    A=LinearInterpolate(A=@n24, B=@n25, Alpha=@n28)
    B=VectorParameter() {parameterName=Glow Color, defaultValue=<rgba 1 0.147 0.1329 1>}.RGB
    Alpha=LinearInterpolate(A=@n29, Alpha=@n30) {constB=0}
   )) {name=Final Color, nodeColor=<rgba 0 1 0.9297 1>}
  NamedRerouteDeclaration(None=MaterialFunctionCall("Value (0-1)"=ScalarParameter() {parameterName=Animation Control}) {materialFunction=&MF_UI_EaseCurves}.EaseInOutCubic) {name=Animation Control, nodeColor=<rgba 0 0.8125 1 1>}
  NamedRerouteUsage()
  NamedRerouteUsage()
 ]
}
