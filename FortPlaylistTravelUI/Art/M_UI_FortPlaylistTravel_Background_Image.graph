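# UI-domain translucent material for a background image in the FortPlaylistTravel UI, judging by its name and path.
# Its only node is a texture sample parameter named Image, which lets material instances supply the texture.
# The graph shows no other processing and no connections to the material outputs.
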Material /FortPlaylistTravelUI/Art/M_UI_FortPlaylistTravel_Background_Image
graph {
 properties {bAutomaticallySetUsageInEditor=false, blendMode=BLEND_Translucent, materialDomain=MD_UI, usedShadingModels=Default Lit}
 nodes [TextureSampleParameter2D() {parameterName=Image}]
}
