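# Translucent UI material for Fort Playlist Travel UI that pairs a parameterized image with a rounded-box mask.
# The Image texture is sampled with MF_UI_UVScaling's scaled-to-fill UVs, using normalized UI UVs and X/Y texture size parameters.
# A signed-distance box on 9-slice UVs takes four per-corner roundness parameters and feeds MF_UI_AntiAliasing for smooth edges.
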Material /FortPlaylistTravelUI/Art/M_UI_FortPlaylistTravel_Image
graph {
 properties {bAutomaticallySetUsageInEditor=false, blendMode=BLEND_Translucent, materialDomain=MD_UI, usedShadingModels=Default Lit}
 nodes [
  n1 = MaterialFunctionCall() {materialFunction=&GetUserInterfaceUV}
  n2 = ScalarParameter() {defaultValue=0.25, parameterName=TL_Roundness}
  n3 = ScalarParameter() {defaultValue=0.25, parameterName=BL_Roundness}
  n4 = ScalarParameter() {defaultValue=0.25, parameterName=TR_Roundness}
  n5 = ScalarParameter() {defaultValue=0.25, parameterName=BR_Roundness}
  MaterialFunctionCall(DistanceField=MaterialFunctionCall(
    UVs=@n1.9-Slice UV
    Size=Constant() {r=1}
    CornerRoundness=MaterialFunctionCall(X=@n2, Y=@n3, Z=@n4, A=@n5) {materialFunction=&MakeFloat4}.Result
   ) {materialFunction=&MF_UI_Box}.SDF) {materialFunction=&MF_UI_AntiAliasing}
  n6 = ScalarParameter() {defaultValue=128, parameterName=X_TexSize}
  n7 = ScalarParameter() {defaultValue=128, parameterName=Y_TexSize}
  TextureSampleParameter2D(UVs=MaterialFunctionCall(UV=@n1.Normalized UV, TextureSize=AppendVector(A=@n6, B=@n7)) {materialFunction=&MF_UI_UVScaling}.Scaled To Fill) {parameterName=Image}
 ]
}
